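# Niagara vehicle cosmetic for the Gray drift trail, drawing trails, particles and rotating wheel-wrap meshes at the car's wheels.
# Three ribbon emitters (main arrow, layered ribbons and a BCKP variant) and a bright sprite emitter stream backward from skeletal-mesh positions, inheriting vehicle movement.
# The BCKP variant adds sine wobble, blended between axes by camera-facing dot product.
# Per-wheel masks disable effects, a frontend-preview flag adjusts ribbon lifetime, and painted-variant overrides are disabled.
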NiagaraSystem /VehicleCosmetics/DriftTrail/DriftTrail_Gray/FX/NS_DriftTrail_Gray_MP
dependencies {
 usedDataInterfaces [&NiagaraDataInterfaceCurve, &NiagaraDataInterfaceCamera, &NiagaraDataInterfaceSkeletalMesh]
 usedDynamicInputs []
 usedModules [
  &self:SystemSpawnScript.SystemScriptSource.SystemScriptGraph.NiagaraNodeAssignment_0.SetVariables_#1
  &SystemState
  &FN_UnpackVehicleFlags
  /Game/Effects/Niagara/Modules/Vehicle/FN_UpdateVehicleAttribs
  /Game/Effects/Niagara/Modules/Vehicle/FN_VehicleFlags_Cheatcodes
 ]
}
emitters {
 BCKP_Ribbons002 {
  emitterSpawnScript [/SetVariables(Emitter.R1 Ribbon=45) {isSetParameters=true, module=SetVariables_#2}]
  emitterUpdateScript [
   &EmitterState()
   &VC_RateQualityHandler(
    Epic Spawn Rate:NiagaraFloat=&MakeCustomFloatFromBool(Bool=$User.bIsFrontendPreview, True Float=30, False Float=45)
    High Spawn Rate=30
    Low Spawn Rate=20
    Medium Spawn Rate=25
    Use Custom Rate?=true
   )
   &VC_SpawnRibbon_v1(Number of Ribbons=2, SpawnRate=150)
  ]
  particleSpawnScript [
   &VC_PaintedVariantsOverride() {enabled=false}
   &VC_InitializeRibbon_v1(
    Custom Offset=true
    Enable Skeletal Position=true
    Facing Axis:ENiagaraOrientationAxis=::ZAxis
    Lifetime=&MakeCustomFloatFromBool(Bool=$User.bIsFrontendPreview, True Float=0.7, False Float=0.4)
    Local Vector Offset=&AddVectorToPosition(Position=<-15 0 -15>, Vector=$User.offset)
    Offset Amount=1
    Ribbon Width=60
   )
   &AddVelocity(Velocity=<-1400 0 0>) {module=AddVelocity001}
   /SetVariables(Particles.01 RibbonTwist=-0.5, Particles.02 RibbonTwist=0.5) {enabled=false, isSetParameters=true, module=SetVariables_#3}
   /VehicleCosmetics/FX/Niagara/Modules/MergeEmitters/VC_ExtraBindingColors(ExtraColors=3, R1 Color:LinearColor=<rgba 1 0.8 0.6 1>, R2 Color:LinearColor=<rgba 1 0.8 0.6 1>) {enabled=false}
  ]
  particleUpdateScript [
   &FN_DisableTireFX(Mask Index Attribute:NiagaraInt32=$Particles.SpawnGroup, Mask Index FL=2, Mask Index FR=3, Mask Index RL=0, Mask Index RR=1)
   &InheritSourceMovement(
    Applied Position Scale=&Multiply_VectorByFloat(
     Vector=<1 1 1>
     Float=&FloatFromCurve(FloatCurve=DataInterface() {propertyValues {Curve {keys [
          {arriveTangent=-0.035, interpMode=RCIM_Cubic, leaveTangent=-0.035, tangentMode=RCTM_User, time=0, value=1}
          {arriveTangent=-0.0604, tangentMode=RCTM_User, time=1, value=0}
         ]}}})
    )
    ApplyForce=false
    ApplyPosition=true
    ApplyPositionFalloff=false
   ) {enabled=false}
   &ParticleState()
   &InheritSourceMovement(
    Applied Position Scale=&Multiply_VectorByFloat(
     Vector=<1 1 1>
     Float=&FloatFromCurve(FloatCurve=DataInterface() {propertyValues {Curve {keys [
          {interpMode=RCIM_Cubic, leaveTangent=-0.0051, tangentMode=RCTM_Break, time=0, value=1}
          {arriveTangent=-0.193, interpMode=RCIM_Cubic, tangentMode=RCTM_Break, time=1, value=0.9}
         ]}}})
    )
    ApplyForce=false
    ApplyPosition=true
    ApplyPositionFalloff=false
   ) {module=InheritSourceMovement001}
   &SolveForcesAndVelocity(Clamp Velocity=false, Limit Acceleration=false)
   &DynamicMaterialParameters(
    Index 0 Param 1=1
    Index 0 Param 2=1
    Module.Index0Param0Write=true
    Module.Index0Param1Write=true
    Module.Index0Param2Write=true
    Module.Index0Param3Write=true
   ) {enabled=false}
   /SetVariables(Particles.R1 Size Scale:NiagaraFloat=&Multiply_Float(A=$Emitter.R1 Ribbon, B=1)) {isSetParameters=true, module=SetVariables_#4}
   &ScaleColor(
    Scale Alpha=&FloatFromCurve(FloatCurve=DataInterface() {propertyValues {Curve {keys [{tangentMode=RCTM_User, time=0, value=0}, {tangentMode=RCTM_User, time=0.35, value=1}]}}})
    Scale RGB=&VectorFromFloat(Value=1)
    ScaleA=true
    ScaleRGB=true
   ) {enabled=false}
   /SetVariables(Particles.1 ribbon scale color:LinearColor=&VC_DynamicScaleColor(
     Color to Scale:LinearColor=$Particles.R1 Color
     RGB Scale:Vector3f=&VectorFromFloat(Value=1.25)
     Alpha Scale:NiagaraFloat=2.1
    )) {isSetParameters=true, module=SetVariables_#5}
   &ScaleRibbonWidth(Ribbon Width Scale=&FloatFromCurve(FloatCurve=DataInterface() {propertyValues {Curve {keys [
         {arriveTangent=4.1785, interpMode=RCIM_Cubic, leaveTangent=4.1785, tangentMode=RCTM_User, time=0, value=0}
         {arriveTangent=-0.051, time=0.5, value=0.8}
         <at 0.8 1>
         {arriveTangent=0.0224, leaveTangent=0.0224, time=0.9735, value=1}
        ]}}}))
   &VC_UpdateRibbon_v1(Enable Skeletal Position=true)
   /SetVariables(
    Particles.01 Position:NiagaraPosition=&AddVectorToPosition(Position=$Particles.Position, Vector=/VehicleCosmetics/FX/Niagara/Modules/VC_ConvertVectorToWorld(Vector:Vector3f=<20 -6 0>))
    Particles.02 position:NiagaraPosition=&AddVectorToPosition(Position=$Particles.Position, Vector=&VC_ConvertVectorToWorld(Vector:Vector3f=<20 6 0>))
   ) {isSetParameters=true, module=SetVariables_#6}
   /SetVariables(Particles.DotProduct:NiagaraFloat=&DotProduct(
     Vector A=/Niagara/DynamicInputs/Math/TransformVectorByMatrix(
      Vector:Vector3f=<1 0 0>
      Transformation Matrix:NiagaraMatrix=$Engine.Owner.SystemLocalToWorld
      Transform Vector as Position:NiagaraBool=false
     )
     Vector B=&NormalizeVector(
      Vector To Normalize=&SubtractVectorFromPosition(
       Position=$Engine.Owner.Position
       Vector=&ReturnCameraPosition(Coordinate Space:ENiagaraCoordinateSpace=::Simulation, Camera Query:NiagaraDataInterfaceCamera=DataInterface())
      )
      Safe Normalize=true
      Safe Fallback Threshold=0
      Safe Fallback Vector=<0 0 0.001>
     )
    )) {isSetParameters=true}
   /SetVariables(Particles.Position=&AddVectorToPosition(
     Position=$Particles.Position
     Vector=&TransformVector(
      Vector=&MakeVector(
       X=&Waveform("Waveform [1]"=::NewEnumerator0, "[1] Frequency"=0.7, "[1] Amplitude Min/Max"=<0 4>, "[1] Scale"=0.5)
       Y=&Lerp_Float(
        A=0
        B=&Sine(
         Normalized Angle=&Add_Float(
          A=&Multiply_Float(A=$Particles.NormalizedAge, B=-2)
          B=&Multiply_Float(A=$Emitter.LoopedAge, B=2)
         )
         Period=1
         Scale=&FloatFromCurve(FloatCurve=DataInterface() {propertyValues {Curve {
             defaultValue=0
             keys [
              {arriveTangent=9.0725, interpMode=RCIM_Cubic, leaveTangent=9.0725, tangentMode=RCTM_User, time=0, value=0}
              {arriveTangent=3.6077, interpMode=RCIM_Cubic, leaveTangent=3.6077, time=0.3792, value=0.8927}
              {arriveTangent=-0.1703, interpMode=RCIM_Cubic, leaveTangent=-0.1703, tangentMode=RCTM_User, time=0.4989, value=1.8}
              {arriveTangent=-0.3525, interpMode=RCIM_Cubic, leaveTangent=-0.3525, tangentMode=RCTM_User, time=0.5483, value=0.4087}
              {arriveTangent=0.3722, interpMode=RCIM_Cubic, leaveTangent=0.3722, tangentMode=RCTM_User, time=0.6854, value=1.6831}
              {arriveTangent=-0.0879, interpMode=RCIM_Cubic, leaveTangent=-0.0879, tangentMode=RCTM_User, time=0.7766, value=0.505}
              {arriveTangent=-0.1448, interpMode=RCIM_Cubic, leaveTangent=-0.1448, tangentMode=RCTM_User, time=0.8125, value=2.125}
              {arriveTangent=0.2336, interpMode=RCIM_Cubic, leaveTangent=0.2336, tangentMode=RCTM_User, time=0.9238, value=1.225}
              <at 0.956 2.7849 RCIM_Cubic>
             ]
            }}})
         Bias=0
        )
        Alpha=&NormalizeFloat(Value=&Power(a=$Particles.DotProduct, b=2), Min=0, Max=1)
       )
       Z=&Lerp_Float(
        A=&Sine(
         Normalized Angle=&Add_Float(
          A=&Multiply_Float(A=$Particles.NormalizedAge, B=-2)
          B=&Multiply_Float(A=$Emitter.LoopedAge, B=2)
         )
         Period=1
         Scale=&FloatFromCurve(
          FloatCurve=DataInterface() {propertyValues {Curve {
             defaultValue=0
             keys [
              {arriveTangent=9.0725, interpMode=RCIM_Cubic, leaveTangent=9.0725, tangentMode=RCTM_User, time=0, value=0}
              {arriveTangent=3.6077, interpMode=RCIM_Cubic, leaveTangent=3.6077, time=0.3792, value=0.8927}
              {arriveTangent=-0.1703, interpMode=RCIM_Cubic, leaveTangent=-0.1703, tangentMode=RCTM_User, time=0.4989, value=1.8}
              {arriveTangent=-0.3525, interpMode=RCIM_Cubic, leaveTangent=-0.3525, tangentMode=RCTM_User, time=0.5483, value=0.4087}
              {arriveTangent=0.3722, interpMode=RCIM_Cubic, leaveTangent=0.3722, tangentMode=RCTM_User, time=0.6854, value=1.6831}
              {arriveTangent=-0.0879, interpMode=RCIM_Cubic, leaveTangent=-0.0879, tangentMode=RCTM_User, time=0.7766, value=0.505}
              {arriveTangent=-0.1448, interpMode=RCIM_Cubic, leaveTangent=-0.1448, tangentMode=RCTM_User, time=0.8125, value=2.125}
              {arriveTangent=0.2336, interpMode=RCIM_Cubic, leaveTangent=0.2336, tangentMode=RCTM_User, time=0.9238, value=1.225}
              <at 0.956 2.7849 RCIM_Cubic>
             ]
            }}}
          Scale Curve=1.5
         )
         Bias=0
        )
        B=0
        Alpha=&NormalizeFloat(Value=&Power(a=$Particles.DotProduct, b=2), Min=0, Max=1)
       )
      )
      Destination Space=::Simulation
      Apply Scale=true
     )
    )) {isSetParameters=true}
   /SetVariables(Particles.CayColor:LinearColor=&MultiplyLinearColor(
     Linear Color=<rgba 1 1 1 1>
     Scale Factor=&MakeVector4FromVectorAndFloat(
      "Vector (RGB)"=<1 1 1>
      "Float (Alpha)"=&FloatFromCurve(FloatCurve=DataInterface() {propertyValues {Curve {
          defaultValue=0
          keys [{interpMode=RCIM_Cubic, tangentMode=RCTM_User, time=0.1927, value=0}, {interpMode=RCIM_Cubic, tangentMode=RCTM_User, time=0.6147, value=1}]
         }}})
     )
    )) {isSetParameters=true}
  ]
  rendererClasses [&NiagaraRibbonRendererProperties]
  renderers [
   {rendererClass=&NiagaraRibbonRendererProperties}
   {rendererClass=&NiagaraRibbonRendererProperties, rendererIndex=1}
   {rendererClass=&NiagaraRibbonRendererProperties, rendererIndex=2}
  ]
 }
 NE_VC_DriftTrail_Sprite_Parent001 {
  emitterUpdateScript [
   &EmitterState()
   &SpawnRate(
    SpawnRate=/VehicleCosmetics/FX/Niagara/Modules/VC_Mask_Floats(
     FrontEnd Value:NiagaraFloat=120
     blsFrontEndPreview:NiagaraBool=$User.bIsFrontendPreview
     SpeedMin Value:NiagaraFloat=120
     SpeedMax Value:NiagaraFloat=120
     Speed Alpha:NiagaraFloat=$System.Speed_Alpha
    )
    Use Spawn Probability=false
   )
  ]
  particleSpawnScript [
   &VC_PaintedVariantsOverride() {enabled=false}
   &InitializeParticle(
    Color=<rgba 0.69 0.42 0.14 1>
    Lifetime Max=0.25
    Lifetime Min=0.12
    Lifetime Mode=::NewEnumerator1
    Sprite Rotation Mode=::NewEnumerator2
    Sprite Size Mode=::NewEnumerator1
    Uniform Sprite Size Max=3
    Uniform Sprite Size Min=1
    UsePositionOffset=false
    Write Material Random=true
   )
   &PartitionParticles()
   &SkeletalMeshLocation(
    Bone / Socket Index=&Add_Integer(A=$Particles.PartitionParticles.Partition, B=4)
    Bone / Socket Sampling Mode=::NewEnumerator5
    Interpolate Mesh Sampling=false
    MaskSpawnGroup=false
    Mesh Sampling Type=::NewEnumerator6
    Sampled Position Offset=&AddVectorToPosition(Position=<-45 0 -2>, Vector=$User.offset)
    Skeletal Mesh=$System.SkeletalMesh
   )
   &AddVelocity(Velocity=&VC_Mask_Vectors(
     FrontEnd Value:Vector3f=<-1300 0 0>
     blsFrontEndPreview:NiagaraBool=$User.bIsFrontendPreview
     SpeedMin Value:Vector3f=<-1300 0 0>
     SpeedMax Value:Vector3f=<-1300 0 0>
     Speed Alpha:NiagaraFloat=$System.Speed_Alpha
    ))
   &ShapeLocation(MaskSpawnGroup=false, Override Seed=false, Sphere Radius=15, UseOwnerScale=true)
  ]
  particleUpdateScript [
   &FN_DisableTireFX(
    Mask Index Attribute:NiagaraInt32=$Particles.PartitionParticles.Partition
    Mask Index FL=2
    Mask Index FR=3
    Mask Index RL=0
    Mask Index RR=1
   )
   &ParticleState()
   &InheritSourceMovement(
    Applied Position Scale=&Multiply_VectorByFloat(
     Vector=<1 1 1>
     Float=&FloatFromCurve(FloatCurve=DataInterface() {propertyValues {Curve {keys [
          {interpMode=RCIM_Cubic, leaveTangent=-0.0051, tangentMode=RCTM_Break, time=0, value=1}
          {arriveTangent=-0.193, interpMode=RCIM_Cubic, tangentMode=RCTM_Break, time=1, value=0.9}
         ]}}})
    )
    ApplyForce=false
    ApplyPosition=true
    ApplyPositionFalloff=false
   )
   &ScaleColor(Scale RGB=<375 225 150>, ScaleA=true, ScaleRGB=true)
   &SolveForcesAndVelocity(Clamp Velocity=false, Limit Acceleration=false)
  ]
  rendererClasses [&NiagaraSpriteRendererProperties]
  renderers [{rendererClass=&NiagaraSpriteRendererProperties}]
 }
 NE_VC_WheelWrap_L {
  emitterUpdateScript [
   &EmitterState(
    Inactive Response=::NewEnumerator1
    Life Cycle Mode=::NewEnumerator1
    Loop Behavior=::NewEnumerator1
    Loop Duration Mode=::NewEnumerator1
    UseLoopDelay=false
   )
   &SpawnBurst_Instantaneous(Spawn Count=2, Use Spawn Probability=false, UseLoopCountLimit=false)
  ]
  particleSpawnScript [
   &VC_PaintedVariantsOverride() {enabled=false}
   &InitializeParticle(
    Color=<rgba 1 0.6379 0.5304 1>
    Mesh Scale Mode=::NewEnumerator4
    Sprite Rotation Mode=::NewEnumerator2
    Sprite Size Mode=::NewEnumerator3
    UsePositionOffset=false
    Write Material Random=true
   )
   &PartitionParticles()
   &SkeletalMeshLocation(
    Bone / Socket Index=&Add_Integer(A=$Particles.PartitionParticles.Partition, B=4)
    Bone / Socket Sampling Mode=::NewEnumerator5
    Interpolate Mesh Sampling=false
    MaskSpawnGroup=false
    Mesh Sampling Type=::NewEnumerator6
    Orientation Sampling=::NewEnumerator1
    Sampled Position Offset Coordinate Space:ENiagaraSkelMeshTransforms=::NewEnumerator2
    Skeletal Mesh=$System.SkeletalMesh
   )
  ]
  particleUpdateScript [
   &FN_DisableTireFX(
    Mask Index Attribute:NiagaraInt32=$Particles.PartitionParticles.Partition
    Mask Index FL=2
    Mask Index FR=3
    Mask Index RL=0
    Mask Index RR=1
   )
   &VC_WheelWrapScale(Wheel Index Attribute:NiagaraInt32=&Add_Integer(A=$Particles.PartitionParticles.Partition, B=2))
   &ScaleColor(
    Scale Alpha=&FloatFromCurve(FloatCurve=DataInterface() {propertyValues {Curve {keys [
         {interpMode=RCIM_Cubic, tangentMode=RCTM_User, time=0, value=0}
         {interpMode=RCIM_Cubic, tangentMode=RCTM_User, time=0.5, value=1}
         {interpMode=RCIM_Cubic, tangentMode=RCTM_User, time=1, value=0}
        ]}}})
    Scale RGB=&VectorFromFloat(Value=$User.WheelIntensity)
    ScaleA=false
    ScaleRGB=true
   ) {enabled=false}
   &ParticleState(Kill Particles When Lifetime Has Elapsed=false)
   &UpdateMeshOrientation(Rotation Coordinate Space=::NewEnumerator2, Rotation Vector=<0 1 0>) {enabled=false}
   &ScaleColor(Scale RGB=&VectorFromFloat(Value=0.7), ScaleA=true, ScaleRGB=true) {module=ScaleColor001}
   &UpdateMeshOrientation(Rotation Coordinate Space=::NewEnumerator2, Rotation Rate=3, Rotation Vector=<0 1 0>) {module=UpdateMeshOrientation001}
   &SkeletalMeshLocation(
    Bone / Socket Index=&Add_Integer(A=$Particles.PartitionParticles.Partition, B=4)
    Bone / Socket Sampling Mode=::NewEnumerator5
    MaskSpawnGroup=false
    Mesh Sampling Type=::NewEnumerator6
    Skeletal Mesh=$System.SkeletalMesh
   ) {module=SkeletalMeshLocation001}
   &SolveForcesAndVelocity(Clamp Velocity=false, Limit Acceleration=false)
  ]
  rendererClasses [&NiagaraMeshRendererProperties]
  renderers [{rendererClass=&NiagaraMeshRendererProperties, rendererIndex=0}]
 }
 Ribbons {
  emitterSpawnScript [/SetVariables(Emitter.R1 Ribbon=45) {isSetParameters=true, module=SetVariables_#2}]
  emitterUpdateScript [
   &EmitterState()
   &VC_RateQualityHandler(
    Epic Spawn Rate:NiagaraFloat=&MakeCustomFloatFromBool(Bool=$User.bIsFrontendPreview, True Float=30, False Float=45)
    High Spawn Rate=30
    Low Spawn Rate=20
    Medium Spawn Rate=25
    Use Custom Rate?=true
   )
   &VC_SpawnRibbon_v1(Number of Ribbons=2, Spawn Time=-1, SpawnRate=$Emitter.Spawn Rate)
  ]
  particleSpawnScript [
   &VC_PaintedVariantsOverride() {enabled=false}
   &VC_InitializeRibbon_v1(
    Custom Offset=true
    Enable Skeletal Position=true
    Facing Axis:ENiagaraOrientationAxis=::ZAxis
    Lifetime=&MakeCustomFloatFromBool(Bool=$User.bIsFrontendPreview, True Float=0.7, False Float=0.5)
    Local Vector Offset=&AddVectorToPosition(Position=<-15 0 -15>, Vector=$User.offset)
    Offset Amount=1
    Ribbon Width=60
   )
   &AddVelocity(Velocity=<-1400 0 0>) {module=AddVelocity001}
   /SetVariables(Particles.01 RibbonTwist=-0.5, Particles.02 RibbonTwist=0.5) {isSetParameters=true, module=SetVariables_#3}
   &VC_ExtraBindingColors(ExtraColors=3, R1 Color:LinearColor=<rgba 1 0.8 0.6 1>, R2 Color:LinearColor=<rgba 1 0.8 0.6 1>)
  ]
  particleUpdateScript [
   &FN_DisableTireFX(Mask Index Attribute:NiagaraInt32=$Particles.SpawnGroup, Mask Index FL=2, Mask Index FR=3, Mask Index RL=0, Mask Index RR=1)
   &InheritSourceMovement(
    Applied Position Scale=&Multiply_VectorByFloat(
     Vector=<1 1 1>
     Float=&FloatFromCurve(FloatCurve=DataInterface() {propertyValues {Curve {keys [
          {arriveTangent=-0.035, interpMode=RCIM_Cubic, leaveTangent=-0.035, tangentMode=RCTM_User, time=0, value=1}
          {arriveTangent=-0.0604, tangentMode=RCTM_User, time=1, value=0}
         ]}}})
    )
    ApplyForce=false
    ApplyPosition=true
    ApplyPositionFalloff=false
   ) {enabled=false}
   &ParticleState()
   &InheritSourceMovement(
    Applied Position Scale=&Multiply_VectorByFloat(
     Vector=<1 1 1>
     Float=&FloatFromCurve(FloatCurve=DataInterface() {propertyValues {Curve {keys [
          {interpMode=RCIM_Cubic, leaveTangent=-0.0051, tangentMode=RCTM_Break, time=0, value=1}
          {arriveTangent=-0.193, interpMode=RCIM_Cubic, tangentMode=RCTM_Break, time=1, value=0.9}
         ]}}})
    )
    ApplyForce=false
    ApplyPosition=true
    ApplyPositionFalloff=false
   ) {module=InheritSourceMovement001}
   &SolveForcesAndVelocity(Clamp Velocity=false, Limit Acceleration=false)
   &DynamicMaterialParameters(
    Index 0 Param 1=1
    Index 0 Param 2=1
    Module.Index0Param0Write=true
    Module.Index0Param1Write=true
    Module.Index0Param2Write=true
    Module.Index0Param3Write=true
   )
   /SetVariables(Particles.R1 Size Scale:NiagaraFloat=&Multiply_Float(A=$Emitter.R1 Ribbon, B=1)) {isSetParameters=true, module=SetVariables_#4}
   &ScaleColor(
    Scale Alpha=&FloatFromCurve(FloatCurve=DataInterface() {propertyValues {Curve {keys [{tangentMode=RCTM_User, time=0, value=0}, {tangentMode=RCTM_User, time=0.35, value=1}]}}})
    Scale RGB=&VectorFromFloat(Value=1.5)
    ScaleA=true
    ScaleRGB=true
   )
   /SetVariables(Particles.1 ribbon scale color:LinearColor=&VC_DynamicScaleColor(
     Color to Scale:LinearColor=$Particles.R1 Color
     RGB Scale:Vector3f=&VectorFromFloat(Value=1.25)
     Alpha Scale:NiagaraFloat=2.1
    )) {isSetParameters=true, module=SetVariables_#5}
   &ScaleRibbonWidth(Ribbon Width Scale=&FloatFromCurve(FloatCurve=DataInterface() {propertyValues {Curve {keys [
         {arriveTangent=4.1785, interpMode=RCIM_Cubic, leaveTangent=4.1785, tangentMode=RCTM_User, time=0, value=0}
         {arriveTangent=-0.051, time=0.5, value=0.8}
         <at 0.8 1>
         {arriveTangent=0.0224, leaveTangent=0.0224, time=1, value=1}
        ]}}}))
   &VC_UpdateRibbon_v1(Enable Skeletal Position=true)
   /SetVariables(
    Particles.01 Position:NiagaraPosition=&AddVectorToPosition(Position=$Particles.Position, Vector=&VC_ConvertVectorToWorld(Vector:Vector3f=<20 -6 0>))
    Particles.02 position:NiagaraPosition=&AddVectorToPosition(Position=$Particles.Position, Vector=&VC_ConvertVectorToWorld(Vector:Vector3f=<20 6 0>))
   ) {isSetParameters=true, module=SetVariables_#6}
  ]
  rendererClasses [&NiagaraRibbonRendererProperties]
  renderers [
   {rendererClass=&NiagaraRibbonRendererProperties}
   {rendererClass=&NiagaraRibbonRendererProperties, rendererIndex=1}
   {rendererClass=&NiagaraRibbonRendererProperties, rendererIndex=2}
  ]
 }
 main_arrow_ribbon {
  emitterUpdateScript [
   &EmitterState()
   &VC_RateQualityHandler(Epic Spawn Rate=90, High Spawn Rate=90, Low Spawn Rate=60, Medium Spawn Rate=60, Use Custom Rate?=true)
   &VC_SpawnRibbon_v1(Number of Ribbons=2, Spawn Time=-1, SpawnRate=$Emitter.Spawn Rate)
  ]
  particleSpawnScript [
   &VC_PaintedVariantsOverride() {enabled=false}
   &VC_InitializeRibbon_v1(
    Color=<rgba 1 0.8 0.6 1>
    Custom Offset=true
    Enable Skeletal Position=true
    Facing Axis:ENiagaraOrientationAxis=::ZAxis
    Lifetime=0.25
    Local Vector Offset=&AddVectorToPosition(Position=<-35 0 -15>, Vector=$User.offset)
    Offset Amount=1
    Ribbon Width=45
   )
   &AddVelocity(Velocity=<-1400 0 0>) {module=AddVelocity001}
  ]
  particleUpdateScript [
   &FN_DisableTireFX(Mask Index Attribute:NiagaraInt32=$Particles.SpawnGroup, Mask Index FL=2, Mask Index FR=3, Mask Index RL=0, Mask Index RR=1)
   &InheritSourceMovement(
    Applied Position Scale=&Multiply_VectorByFloat(
     Vector=<1 1 1>
     Float=&FloatFromCurve(FloatCurve=DataInterface() {propertyValues {Curve {keys [
          {arriveTangent=-0.035, interpMode=RCIM_Cubic, leaveTangent=-0.035, tangentMode=RCTM_User, time=0, value=1}
          {arriveTangent=-0.0604, tangentMode=RCTM_User, time=1, value=0}
         ]}}})
    )
    ApplyForce=false
    ApplyPosition=true
    ApplyPositionFalloff=false
   ) {enabled=false}
   &ParticleState()
   &InheritSourceMovement(
    Applied Position Scale=&Multiply_VectorByFloat(
     Vector=<1 1 1>
     Float=&FloatFromCurve(FloatCurve=DataInterface() {propertyValues {Curve {keys [
          {interpMode=RCIM_Cubic, leaveTangent=-0.0051, tangentMode=RCTM_Break, time=0, value=1}
          {arriveTangent=-0.193, interpMode=RCIM_Cubic, tangentMode=RCTM_Break, time=1, value=0.9}
         ]}}})
    )
    ApplyForce=false
    ApplyPosition=true
    ApplyPositionFalloff=false
   ) {module=InheritSourceMovement001}
   &SolveForcesAndVelocity(Clamp Velocity=false, Limit Acceleration=false)
   &DynamicMaterialParameters(
    Index 0 Param 1=1
    Index 0 Param 2=1
    Module.Index0Param0Write=true
    Module.Index0Param1Write=true
    Module.Index0Param2Write=true
    Module.Index0Param3Write=true
   ) {enabled=false}
   &ScaleColor(Scale RGB=&VectorFromFloat(Value=1.25), ScaleA=true, ScaleRGB=true)
   &VC_UpdateRibbon_v1(Enable Skeletal Position=true)
  ]
  rendererClasses [&NiagaraRibbonRendererProperties]
  renderers [{rendererClass=&NiagaraRibbonRendererProperties}]
 }
}
summary {
 systemName=NS_DriftTrail_Gray_MP
 emitters [
  {emitterName=NE_VC_WheelWrap_L, rendererClasses [&NiagaraMeshRendererProperties]}
  {emitterName=main_arrow_ribbon, rendererClasses [&NiagaraRibbonRendererProperties]}
  {emitterName=NE_VC_DriftTrail_Sprite_Parent001, rendererClasses [&NiagaraSpriteRendererProperties]}
  {emitterName=BCKP_Ribbons002, rendererClasses [&NiagaraRibbonRendererProperties]}
  {emitterName=Ribbons, rendererClasses [&NiagaraRibbonRendererProperties]}
 ]
 userVariables [
  User.bIsFrontendPreview {type=NiagaraBool}
  User.bIsPainted {type=NiagaraBool}
  User.blsFrontEndPreview {type=NiagaraBool}
  User.ColorVariantID {type=NiagaraInt32}
  User.IDColor {type=NiagaraInt32}
  User.IsPainted {type=NiagaraBool}
  User.offset {type=Vector3f, description="offset from wheels(set through cosmetic template)", defaultValue {struct=Vector3f, value=<0 0 0>}}
  User.PaintedColor {type=LinearColor, description="", defaultValue {struct=LinearColor, value=<rgba 1 1 1 1>}}
  User.PaintedScalar {type=NiagaraFloat, defaultValue=1, description=""}
  User.VariableForPainted {type=LinearColor, description="", defaultValue {struct=LinearColor, value=<rgba 1 1 1 1>}}
  User.VehicleFlags {type=NiagaraInt32, defaultValue=806289408}
  User.Wheel_Radius_Array {
   type=NiagaraDataInterfaceArrayFloat
   defaultValue {
    struct=NiagaraExt_VariableValue_DataInterface
    value {
     dataInterface=&self:NiagaraDataInterfaceArrayFloat_0
     dataInterfaceClass=&NiagaraDataInterfaceArrayFloat
    }
   }
  }
  User.Wheel_Width_Array {
   type=NiagaraDataInterfaceArrayFloat
   defaultValue {
    struct=NiagaraExt_VariableValue_DataInterface
    value {
     dataInterface=&self:NiagaraDataInterfaceArrayFloat_1
     dataInterfaceClass=&NiagaraDataInterfaceArrayFloat
    }
   }
  }
  User.WheelIntensity {type=NiagaraFloat, defaultValue=4.5, description=""}
  User.WheelSize {type=NiagaraFloat, defaultValue=0.1, description=""}
 ]
}
