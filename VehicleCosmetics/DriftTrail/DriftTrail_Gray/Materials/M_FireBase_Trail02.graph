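# Masked unlit fire material for the Gray drift trail cosmetic, built for Niagara ribbons and mesh particles.
# UV gradient masks shape the trail, while panning distortion noise warps rotated gradients into a streak mask.
# Two panning flow normal textures, used as noise, are thresholded with that streak to break up the flame.
# These masks blend EdgeColor and CoreColor, tinted by particle color, into MF_FX_EmissiveAdjustments, with particle alpha scaling intensity.
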Material /VehicleCosmetics/DriftTrail/DriftTrail_Gray/Materials/M_FireBase_Trail02
graph {
 properties {
  bAutomaticallySetUsageInEditor=false
  bUsedWithNiagaraMeshParticles=true
  bUsedWithNiagaraRibbons=true
  blendMode=BLEND_Masked
  shadingModel=MSM_Unlit
  usedShadingModels=Unlit
 }
 nodes [
  MaterialExpressionVertexColor_0 = ParticleColor()
  n1 = TextureCoordinate()
  n2 = ComponentMask(None=@n1) {r=true}
  MaterialExpressionConstantClamp_5 = Saturate(None=ConstantBiasScale(None=OneMinus(None=@n2)) {bias=-0.005, scale=2})
  MaterialExpressionConstantClamp_6 = Saturate(None=ConstantBiasScale(None=@n2) {bias=-0.005, scale=6})
  n3 = ComponentMask(None=@n1) {g=true}
  n4 = Multiply(A=@n3, B=OneMinus(None=@n3))
  n5 = Constant() {r=5}
  MaterialExpressionConstantClamp_3 = Saturate(None=Multiply(A=Multiply(A=@MaterialExpressionConstantClamp_5, B=@MaterialExpressionConstantClamp_6), B=Multiply(A=@n4, B=@n5)))
  n6 = Constant()
  n7 = TextureCoordinate()
  n8 = Add(A=@n7, B=Constant2Vector() {g=-0.2})
  n9 = Constant() {r=0.9}
  n10 = Constant() {r=0.5}
  n11 = Panner(Coordinate=TextureCoordinate() {vTiling=0.125}, Time=@n10) {speedY=0.5}
  n12 = TextureSample(UVs=Panner(Coordinate=TextureCoordinate() {uTiling=0.25, vTiling=0.25}) {speedX=-0.125, speedY=0.01}) {texture=/DealershipSharedLibrary/FX/Textures/Noise/T_VCVFX_Flow02_N}
  n13 = Multiply(A=Constant() {r=0.05}, B=AppendVector(A=@n12.R, B=@n12.R))
  n14 = TextureSample(UVs=Panner(Coordinate=Add(A=@n11, B=@n13)) {speedX=-1}) {texture=/DealershipSharedLibrary/FX/Textures/Noise/T_VCVFX_DistortNoise02_RGB}
  n15 = Panner(Coordinate=TextureCoordinate() {uTiling=1.25, vTiling=-0.125}, Time=@n10) {speedY=-0.5}
  n16 = TextureSample(UVs=Panner(Coordinate=Add(A=@n15, B=@n13)) {speedX=-1}) {texture=&T_VCVFX_DistortNoise02_RGB}
  n17 = Multiply(A=Constant() {r=0.155}, B=ComponentMask(None=Add(A=@n14.RGB, B=@n16.RGB)) {g=true, r=true})
  n18 = ComponentMask(None=Add(A=Rotator(Coordinate=@n8, Time=@n9) {speed=3.14}, B=@n17)) {g=true}
  MaterialExpressionConstantClamp_7 = Saturate(None=ConstantBiasScale(None=@n18) {bias=-0.015, scale=1})
  MaterialExpressionConstantClamp_8 = Saturate(None=ConstantBiasScale(None=OneMinus(None=@n18)) {bias=-0.01, scale=1})
  n19 = Add(A=@n7, B=Constant2Vector() {g=0.2})
  n20 = Constant() {r=0.1}
  n21 = ComponentMask(None=Add(A=@n17, B=Rotator(Coordinate=@n19, Time=@n20) {speed=3.14})) {g=true}
  MaterialExpressionConstantClamp_9 = Saturate(None=ConstantBiasScale(None=@n21) {bias=-0.015, scale=1})
  MaterialExpressionConstantClamp_10 = Saturate(None=ConstantBiasScale(None=OneMinus(None=@n21)) {bias=-0.01, scale=1})
  n22 = Multiply(
   A=Multiply(A=@MaterialExpressionConstantClamp_7, B=@MaterialExpressionConstantClamp_8)
   B=Multiply(A=@MaterialExpressionConstantClamp_9, B=@MaterialExpressionConstantClamp_10)
  )
  MaterialExpressionConstantClamp_14 = Saturate(None=Multiply(A=@n22, B=Constant() {r=20}))
  n23 = TextureCoordinate()
  n24 = ComponentMask(None=@n23) {r=true}
  MaterialExpressionConstantClamp_16 = Saturate(None=ConstantBiasScale(None=OneMinus(None=@n24)) {bias=-0.005, scale=4})
  MaterialExpressionConstantClamp_17 = Saturate(None=ConstantBiasScale(None=@n24) {bias=-0.005, scale=6})
  n25 = ComponentMask(None=@n23) {g=true}
  n26 = Multiply(A=@n25, B=OneMinus(None=@n25))
  n27 = Constant() {r=5}
  MaterialExpressionConstantClamp_15 = Saturate(None=Multiply(A=Multiply(A=@MaterialExpressionConstantClamp_16, B=@MaterialExpressionConstantClamp_17), B=Multiply(A=@n26, B=@n27)))
  n28 = Multiply(A=@MaterialExpressionConstantClamp_14, B=@MaterialExpressionConstantClamp_15)
  n29 = TextureCoordinate() {uTiling=0.25, vTiling=0.25}
  n30 = Constant() {r=-0.05}
  n31 = Panner(
   Coordinate=Multiply(A=Rotator(Coordinate=@n29, Time=@n30) {speed=3.14}, B=Constant2Vector() {g=2, r=4})
   Time=Multiply(A=Time(), B=Constant() {r=0.48})
  ) {speedX=-2, speedY=0.25}
  n32 = TextureSample(UVs=@n31) {samplerType=SAMPLERTYPE_Normal, texture=/DealershipSharedLibrary/FX/Textures/Normal/T_VCVFX_FlowN03_N}
  n33 = TextureSample(UVs=@n31) {samplerType=SAMPLERTYPE_Normal, texture=/DealershipSharedLibrary/FX/Textures/Noise/T_VCVFX_Water02_N}
  n34 = Constant() {r=0.35}
  n35 = ComponentMask(None=Multiply(A=LinearInterpolate(A=@n32.RGB, B=@n33.RGB, Alpha=@n34), B=Constant() {r=2.5})) {g=true, r=true}
  n36 = Constant() {r=1}
  n37 = ComponentMask(None=Add(A=AppendVector(A=@n6, B=@n28), B=Multiply(A=@n35, B=@n36))) {g=true}
  n38 = Constant() {r=0.5}
  MaterialExpressionConstantClamp_13 = Saturate(None=Multiply(A=Subtract(A=@n37, B=@n38), B=Constant() {r=2}))
  MaterialExpressionConstantClamp_12 = Saturate(None=Multiply(A=@MaterialExpressionConstantClamp_3, B=@MaterialExpressionConstantClamp_13))
  n39 = Multiply(A=@MaterialExpressionVertexColor_0.A, B=@MaterialExpressionConstantClamp_12)
  n40 = Add(A=@n32.RGB, B=@n33.RGB)
  MaterialExpressionConstantClamp_11 = Saturate(None=ConstantBiasScale(None=Desaturation(None=@n40) {luminanceFactors=<rgba 0.3 0.59 0.11 0>}) {bias=-0.1, scale=3})
  n41 = Multiply(A=@MaterialExpressionConstantClamp_11, B=@MaterialExpressionConstantClamp_3)
  n42 = Add(A=@n39, B=Multiply(A=@n39, B=@n41))
  MaterialExpressionConstantClamp_0 = Saturate(None=@n42)
  n43 = VectorParameter() {parameterName=EdgeColor, defaultValue=<rgba 0.4936 0.005 0.009 1>}
  n44 = VectorParameter() {parameterName=CoreColor, defaultValue=<rgba 1 0.6942 0.3012 1>}
  n45 = TextureCoordinate()
  n46 = Constant() {r=0.75}
  n47 = ComponentMask(None=Add(A=Rotator(Coordinate=@n45, Time=@n46) {speed=3.14}, B=Constant() {r=-0.25})) {g=true}
  MaterialExpressionConstantClamp_2 = Saturate(None=ConstantBiasScale(None=OneMinus(None=@n47)) {bias=-0.1, scale=1})
  MaterialExpressionConstantClamp_4 = Saturate(None=Multiply(A=Constant() {r=150}, B=@n22))
  n48 = Multiply(A=@MaterialExpressionConstantClamp_3, B=@MaterialExpressionConstantClamp_4)
  n49 = Constant() {r=0.35}
  n50 = Multiply(
   A=LinearInterpolate(A=@n43.RGB, B=Multiply(A=@n44.RGB, B=@MaterialExpressionConstantClamp_2), Alpha=LinearInterpolate(A=@n41, B=@n48, Alpha=@n49))
   B=@MaterialExpressionVertexColor_0.RGB
  )
  MaterialExpressionConstantClamp_1 = Saturate(None=ConstantBiasScale(None=@MaterialExpressionConstantClamp_13) {bias=-0.6, scale=8})
  MaterialFunctionCall(Emissive=LinearInterpolate(A=Multiply(A=@n43.RGB, B=@MaterialExpressionVertexColor_0.RGB), B=Multiply(A=@n50, B=@n42), Alpha=@MaterialExpressionConstantClamp_1)) {materialFunction=&MF_FX_EmissiveAdjustments}
 ]
}
